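# Prototype Niagara projectile effect for SinisterMelon, combining a brief sprite burst with two mesh trail emitters.
# DirectionalBurst spawns short-lived sprites on a scaled torus, with emitter-age curves shaping location and cone direction, then drag and fade.
# The near-identical Wide_Trail emitters each spawn one mesh particle, oriented differently, that darkens and fades over its lifetime.
# User.ColorVariant indexes each emitter's color array; the other user parameters are unreferenced in the listed modules.
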NiagaraSystem /SinisterMelon/Gameplay/SinisterMelon/FX/Proto/NS_SinisterMelon_Prj_Proto
dependencies {
 usedDataInterfaces [&NiagaraDataInterfaceArrayColor, &NiagaraDataInterfaceCurve, &NiagaraDataInterfaceVectorCurve]
 usedDynamicInputs []
 usedModules [&SystemState]
}
emitters {
 DirectionalBurst {
  emitterUpdateScript [
   &EmitterState(Life Cycle Mode=::NewEnumerator1, Loop Behavior=::NewEnumerator1, Loop Delay=0.02, Loop Duration=0.15, UseLoopDelay=true)
   &SpawnRate(SpawnRate=300, Use Spawn Probability=false)
   &SpawnBurst_Instantaneous(Spawn Count=75, Use Spawn Probability=false, UseLoopCountLimit=false) {enabled=false}
  ]
  particleSpawnScript [
   &InitializeParticle(
    Color=&SelectLinearColorFromArray(
     Array Sampling Mode=::NewEnumerator1
     Direct Array Index=$User.ColorVariant
     Color Selection Array=DataInterface() {propertyValues {ColorData [<rgba 0 0.4025 1.5 0.4>, <rgba 0.9531 0.0728 0.076 0.6>, <rgba 1 0.3072 0 0.4>, <rgba 0.3522 0 1 0.5>]}}
     Random Seed=?
    )
    Lifetime Max=0.08
    Lifetime Min=0.05
    Lifetime Mode=::NewEnumerator1
    Sprite Size Mode=::NewEnumerator1
    Uniform Sprite Size Max=3
    Uniform Sprite Size Min=1.5
    UsePositionOffset=false
    Write Material Random=true
   )
   &ShapeLocation(
    Handle Radius=35
    Large Radius=150
    MaskSpawnGroup=false
    Non Uniform Scale=<-1 -0.8 1>
    Override Seed=false
    Rotation Mode=::NewEnumerator1
    Shape Primitive=::NewEnumerator3
    U Distribution:NiagaraFloat=&FloatFromCurve(FloatCurve=DataInterface() {propertyValues {Curve {keys [<at 0 1>,<at 0.2 0.5>]}}}, CurveIndex=$Emitter.Age)
    UseOwnerScale=false
    V Distribution=0.8929
   )
   &AddVelocity(
    Cone Angle=5
    Cone Axis=&MakeVector(
     X=0
     Y=&FloatFromCurve(FloatCurve=DataInterface() {propertyValues {Curve {keys [<at 0 1>,<at 0.1 -0.1>]}}}, CurveIndex=$Emitter.Age)
     Z=&FloatFromCurve(FloatCurve=DataInterface() {propertyValues {Curve {keys [<at 0 0.1>,<at 0.1 -0.25>]}}}, CurveIndex=$Emitter.Age)
    )
    Distribution Along Cone Axis=0.2
    Inner Cone Angle=10
    Module.UseConeFalloff=false
    Module.UseVelDistribution=true
    Override Seed=false
    Velocity Mode=::NewEnumerator2
    Velocity Speed=&RandomRangeFloat(Minimum=50, Maximum=200)
   )
   &InheritVelocity(Speed Limit=true, UseSpeedThreshold=false) {enabled=false}
   &ApplyInitialForces() {enabled=false}
  ]
  particleUpdateScript [
   &ParticleState()
   &GravityForce() {enabled=false}
   &Drag(Drag=&RandomRangeFloat(Minimum=4, Maximum=5), Use Linear Drag=true, Use Rotational Drag=false)
   &ScaleColor(
    Scale Alpha=&FloatFromCurve(FloatCurve=DataInterface() {propertyValues {Curve {keys [<at -0.0046 1.0113 RCIM_Cubic>,<at 1.0011 0.002 RCIM_Cubic>]}}}, Scale Curve=0.2)
    Scale RGB=<2 2 2>
    ScaleA=true
    ScaleRGB=true
   )
   &SolveForcesAndVelocity(Clamp Velocity=false, Limit Acceleration=false)
   &ScaleSpriteSizeBySpeed(Max Scale Factor=<1 5>, Min Scale Factor=<0.75 0.3>, Velocity Threshold=1500) {enabled=false}
  ]
  rendererClasses [&NiagaraSpriteRendererProperties]
  renderers [{rendererClass=&NiagaraSpriteRendererProperties}]
 }
 Wide_Trail {
  emitterUpdateScript [
   &EmitterState()
   &SpawnBurst_Instantaneous(Spawn Count=1, Use Spawn Probability=false)
  ]
  particleSpawnScript [
   /Niagara/Modules/Spawn/Initialization/InitializeParticle(
    Color=&SelectLinearColorFromArray(
     Array Sampling Mode=::NewEnumerator1
     Direct Array Index=$User.ColorVariant
     Color Selection Array=DataInterface() {propertyValues {ColorData [<rgba 1.5 1.334 0.0067 0.4>, <rgba 0.9449 0.9531 0.1156 1>, <rgba 1 0.3072 0 0.4>, <rgba 0.3522 0 1 0.5>]}}
     Random Seed=?
    )
    Lifetime=10
    Mesh Scale=<20 20 20>
    Module.Write Color=true
    Module.Write Lifetime=true
    Module.Write Mass=true
    Module.Write Material Random=true
    Module.Write Position=true
    Module.Write Scale=true
    Module.Write SpriteRotation=false
    Module.Write SpriteSize=false
    Position=&SimulationPosition()
   )
   /Niagara/Modules/Spawn/Initialization/V2/InitializeParticle(
    Color=&SelectLinearColorFromArray(
     Array Sampling Mode=::NewEnumerator1
     Direct Array Index=$User.ColorVariant
     Color Selection Array=DataInterface() {propertyValues {ColorData [<rgba 13.7817 0 30 0.5>, <rgba 15 0.4582 0 0.5>]}}
     Random Seed=?
    )
    Lifetime=0.15
    Mass Mode=::NewEnumerator1
    Mesh Scale=<30 30 30>
    Mesh Scale Mode=::NewEnumerator4
    Position=&SimulationPosition()
    Position Mode=::NewEnumerator1
    Sprite Size Mode=::NewEnumerator4
    UsePositionOffset=false
    Write Material Random=true
   ) {enabled=false, module=InitializeParticle001}
   &InitialMeshOrientation(
    Mesh Axis To Orient=<0 0 0>
    Mesh Orientation Mode=::NewEnumerator2
    Orientation Coordinate Space=::Local
    Orientation Vector=<45 0 0>
    Use Rotation Vector=false
   )
   &SystemLocation(MaskSpawnGroup=false)
  ]
  particleUpdateScript [
   &ParticleState()
   &ScaleMeshSize(Scale Factor=&Multiply_VectorByFloat(
     Vector=<1 1 1>
     Float=&FloatFromCurve(FloatCurve=DataInterface() {propertyValues {Curve {keys [
          {arriveTangent=1.0119, interpMode=RCIM_Cubic, leaveTangent=1.0119, tangentMode=RCTM_User, time=1, value=1}
          {arriveTangent=0.0224, interpMode=RCIM_Cubic, leaveTangent=0.0224, tangentMode=RCTM_User, time=1, value=1}
         ]}}})
    ))
   &ScaleColor(
    Scale Alpha=&FloatFromCurve(FloatCurve=DataInterface() {propertyValues {Curve {keys [{arriveTangent=6.3626, interpMode=RCIM_Cubic, time=0, value=1}, {interpMode=RCIM_Cubic, leaveTangent=-0.6839, time=1, value=0}]}}})
    Scale RGB=&VectorFromCurve(VectorCurve=DataInterface() {propertyValues {XCurve {keys [<at 0 1 RCIM_Cubic>,<at 1 0>]}, YCurve {keys [<at 0 1 RCIM_Cubic>,<at 1 0>]}, ZCurve {keys [<at 0 1 RCIM_Cubic>,<at 1 0>]}}})
    ScaleA=true
    ScaleRGB=true
   )
  ]
  rendererClasses [&NiagaraMeshRendererProperties]
  renderers [
   {rendererClass=&NiagaraMeshRendererProperties, rendererIndex=0}
   {rendererClass=&NiagaraMeshRendererProperties, rendererIndex=1}
  ]
 }
 Wide_Trail001 {
  emitterUpdateScript [
   &EmitterState()
   &SpawnBurst_Instantaneous(Spawn Count=1, Use Spawn Probability=false)
  ]
  particleSpawnScript [
   /Niagara/Modules/Spawn/Initialization/InitializeParticle(
    Color=&SelectLinearColorFromArray(
     Array Sampling Mode=::NewEnumerator1
     Direct Array Index=$User.ColorVariant
     Color Selection Array=DataInterface() {propertyValues {ColorData [<rgba 1.5 1.334 0.0067 0.4>, <rgba 0.9449 0.9531 0.1156 1>, <rgba 1 0.3072 0 0.4>, <rgba 0.3522 0 1 0.5>]}}
     Random Seed=?
    )
    Lifetime=10
    Mesh Scale=<20 20 20>
    Module.Write Color=true
    Module.Write Lifetime=true
    Module.Write Mass=true
    Module.Write Material Random=true
    Module.Write Position=true
    Module.Write Scale=true
    Module.Write SpriteRotation=false
    Module.Write SpriteSize=false
    Position=&SimulationPosition()
   )
   /Niagara/Modules/Spawn/Initialization/V2/InitializeParticle(
    Color=&SelectLinearColorFromArray(
     Array Sampling Mode=::NewEnumerator1
     Direct Array Index=$User.ColorVariant
     Color Selection Array=DataInterface() {propertyValues {ColorData [<rgba 13.7817 0 30 0.5>, <rgba 15 0.4582 0 0.5>]}}
     Random Seed=?
    )
    Lifetime=0.15
    Mass Mode=::NewEnumerator1
    Mesh Scale=<30 30 30>
    Mesh Scale Mode=::NewEnumerator4
    Position=&SimulationPosition()
    Position Mode=::NewEnumerator1
    Sprite Size Mode=::NewEnumerator4
    UsePositionOffset=false
    Write Material Random=true
   ) {enabled=false, module=InitializeParticle001}
   &InitialMeshOrientation(
    Mesh Axis To Orient=<0 0 90>
    Mesh Orientation Mode=::NewEnumerator2
    Orientation Coordinate Space=::Local
    Orientation Vector=<45 0 0>
    Use Rotation Vector=false
   )
   &SystemLocation(MaskSpawnGroup=false)
  ]
  particleUpdateScript [
   &ParticleState()
   &ScaleMeshSize(Scale Factor=&Multiply_VectorByFloat(
     Vector=<1 1 1>
     Float=&FloatFromCurve(FloatCurve=DataInterface() {propertyValues {Curve {keys [
          {arriveTangent=1.0119, interpMode=RCIM_Cubic, leaveTangent=1.0119, tangentMode=RCTM_User, time=1, value=1}
          {arriveTangent=0.0224, interpMode=RCIM_Cubic, leaveTangent=0.0224, tangentMode=RCTM_User, time=1, value=1}
         ]}}})
    ))
   &ScaleColor(
    Scale Alpha=&FloatFromCurve(FloatCurve=DataInterface() {propertyValues {Curve {keys [{arriveTangent=6.3626, interpMode=RCIM_Cubic, time=0, value=1}, {interpMode=RCIM_Cubic, leaveTangent=-0.6839, time=1, value=0}]}}})
    Scale RGB=&VectorFromCurve(VectorCurve=DataInterface() {propertyValues {XCurve {keys [<at 0 1 RCIM_Cubic>,<at 1 0>]}, YCurve {keys [<at 0 1 RCIM_Cubic>,<at 1 0>]}, ZCurve {keys [<at 0 1 RCIM_Cubic>,<at 1 0>]}}})
    ScaleA=true
    ScaleRGB=true
   )
  ]
  rendererClasses [&NiagaraMeshRendererProperties]
  renderers [
   {rendererClass=&NiagaraMeshRendererProperties, rendererIndex=0}
   {rendererClass=&NiagaraMeshRendererProperties, rendererIndex=1}
  ]
 }
}
summary {
 systemName=NS_SinisterMelon_Prj_Proto
 emitters [
  {emitterName=Wide_Trail, rendererClasses [&NiagaraMeshRendererProperties]}
  {emitterName=DirectionalBurst, rendererClasses [&NiagaraSpriteRendererProperties]}
  {emitterName=Wide_Trail001, rendererClasses [&NiagaraMeshRendererProperties]}
 ]
 userVariables [
  User.bIsFrontEnd {type=NiagaraBool}
  User.bIsFrontEndPreview {type=NiagaraBool}
  User.bisFullyDeployed {type=NiagaraBool}
  User.Color {type=LinearColor, description="", defaultValue {struct=LinearColor, value=<rgba 1 1 1 1>}}
  User.ColorVariant {type=NiagaraInt32, defaultValue=1}
  User.CustomColorIntensity {type=NiagaraFloat, defaultValue=0, description=""}
  User.ForwardDot {type=NiagaraFloat, defaultValue=0, description=""}
  User.RibbonIDOffset {type=NiagaraInt32}
  User.RightDot {type=NiagaraFloat, defaultValue=0, description=""}
  User.SpawnSpacing_ribbon {type=NiagaraFloat, defaultValue=150, description=""}
  User.TrailOpacity {type=NiagaraFloat, defaultValue=0, description=""}
  User.UseCustomColor {type=NiagaraFloat, defaultValue=0, description=""}
 ]
}
